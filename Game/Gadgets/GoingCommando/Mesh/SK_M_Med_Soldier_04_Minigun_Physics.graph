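# Physics asset for the SK_M_Med_Soldier_04_Minigun skeletal mesh, the minigun in the GoingCommando gadget content.
# It holds a single body on the Handle_Break bone, shaped as one long box offset from the bone and rotated to align with it.
# There are no constraints, so it gives the weapon one simple collision volume rather than an articulated ragdoll.
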PhysicsAsset /Game/Gadgets/GoingCommando/Mesh/SK_M_Med_Soldier_04_Minigun_Physics
bodies {Handle_Break [{shapeType=Box, extentX=27, extentY=140.5, extentZ=27, center=<50.2603 0 0>, rotation=<rot 0 -90 0>}]}
constraints []
